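; Urban street light building prop for the Apollo environment, with spot lights and a light cone mesh.
; The construction script treats the main SpotLight as master, copying its intensity, falloff exponent, color and cone angles to every other spot light component.
; It supplies the static mesh and light cone to dynamic destruction mesh lists.
; Day phase changes call the parent handler, then print the new phase as debug text, off screen.
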(event EventOnDayPhaseChanged
  (params (CurrentDayPhase :type EFortDayPhase%20Enum :default "Morning") (PreviousDayPhase :type EFortDayPhase%20Enum :default "Morning") (bAtCreation :type Boolean))
  (|CallParentFunction CurrentDayPhase PreviousDayPhase bAtCreation)
  (Development|PrintString (Utilities|String|Append "Time has changed " (Utilities|String|EnumtoString CurrentDayPhase)) false))

(fn GetDynamicDestructionStaticMeshComponents
  (return (Utilities|Array|MakeArray (Variables|Debug|GetStaticMeshComponent)) (Utilities|Array|MakeArray (Variables|Debug|GetStaticMeshComponent) (Variables|Default|GetSM_Street_UrbanLightCone)) (Utilities|Array|MakeArray (Variables|Debug|GetStaticMeshComponent))))

(fn ConstructionScript
  (bind _returnvalue (Actor|GetComponentsByClass "/Script/Engine.SpotLightComponent"))
  (|CallParentFunction)
  (Utilities|Array|RemoveItem _returnvalue (Variables|Default|GetSpotLight))
  (for _array_element _returnvalue
    (Rendering|Components|Light|SetIntensity _array_element (Variables|/Script/Engine.LightComponentBase|GetIntensity (Variables|Default|GetSpotLight)))
    (Rendering|Lighting|SetLightFalloffExponent _array_element (Variables|/Script/Engine.PointLightComponent|GetLightFalloffExponent (Variables|Default|GetSpotLight)))
    (Rendering|Components|Light|SetLightColor _array_element (Rendering|Components|Light|GetLightColor (Variables|Default|GetSpotLight)) true)
    (Rendering|Lighting|SetInnerConeAngle _array_element (Variables|/Script/Engine.SpotLightComponent|GetInnerConeAngle (Variables|Default|GetSpotLight)))
    (Rendering|Lighting|SetOuterConeAngle _array_element (Variables|/Script/Engine.SpotLightComponent|GetOuterConeAngle (Variables|Default|GetSpotLight)))))
